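# Material for the CountyFrog drum kit cosmetic, using the 2023 anime character shader functions (Pre, Calculations, Post) with a parameter-driven eyelid blink.
# Drumkit_Blink offsets floored gradients from UV1's V axis to close upper and lower lids, confined by an eyelid mask texture.
# A slightly offset second threshold adds a black lid edge, and the blue lids replace both lit and shaded colors before Post.
# A time-based sine chain is left unconnected.
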Material /SparksCosmetics/Instrument/DrumKit/CountyFrog/Materials/M_CountyFrog_DrumKit
graph {
 properties {
  bAutomaticallySetUsageInEditor=false
  bUseMaterialAttributes=true
  bUsedWithMeshDeformer=true
  bUsedWithSkeletalMesh=true
  bUsedWithStaticLighting=true
  pixelDepthOffsetMode=PDOM_Legacy
  shadingModel=MSM_Cloth
  usedShadingModels=Cloth
 }
 nodes [
  n1 = Saturate(None=Multiply(A=Time() {bOverride_Period=true, period=3}) {constB=2})
  Divide(A=Sine(None=Add(A=@n1) {constB=-0.2})) {constB=1.7}
  n2 = MaterialFunctionCall!(CustomUV) {materialFunction=/Game/Global/Materials/Character_Parent_2023/Anime_Parent_2023/MaterialFunctions/MFC_AnimeShader_2023_Pre}
  n3 = MaterialFunctionCall!(CustomLightVector, OverrideNormal_WS, STT=@n2.STT, Normal_Map=@n2.STT, DFL=@n2.STT, FX=@n2.STT) {materialFunction=/Game/Global/Materials/Character_Parent_2023/Anime_Parent_2023/MaterialFunctions/MFC_AnimeShader_2023_Calculations}
  n4 = Constant3Vector() {constant=<rgba 0.1845 0.5333 0.7157 1>}
  n5 = ComponentMask(None=TextureCoordinate() {coordinateIndex=1}) {g=true}
  n6 = NamedRerouteUsage()
  n7 = Add(A=OneMinus(None=@n5), B=@n6) {constB=0.25}
  n8 = Add(A=@n7) {constB=-0.04}
  n9 = Add(A=@n5, B=@n6) {constB=0.5}
  n10 = Add(A=@n9) {constB=-0.04}
  n11 = Saturate(None=Round(None=@n5))
  n12 = Saturate(None=LinearInterpolate(A=Floor(None=@n8), B=Floor(None=@n10), Alpha=@n11))
  n13 = Floor(None=@n7)
  n14 = Floor(None=@n9)
  n15 = Multiply(
   A=Saturate(None=LinearInterpolate(A=@n13, B=@n14, Alpha=@n11))
   B=TextureSample() {texture=/SparksCosmetics/Instrument/DrumKit/CountyFrog/Textures/T_CountyFrog_DrumKit_02_Eyelid_Mask}.R
  )
  n16 = Constant3Vector() {constant=<rgba 0.117 0.3916 0.6376 1>}
  MaterialFunctionCall!(
   Shading_Gradient=@n3.Shading_Gradient
   Threshold_Gradient=@n3.Shading_Gradient
   Specular_Gradient=@n3.Shading_Gradient
   Sheen_Mask=@n3.Shading_Gradient
   Thickness=@n3.Shading_Gradient
   DFL_Mask=@n3.Shading_Gradient
   FX=@n3.Shading_Gradient
   Metallic_Gradient=@n3.Shading_Gradient
   RimLight_Gradient=@n3.Shading_Gradient
   NormalBuffer=@n3.Shading_Gradient
   --------
   CINE_LightSources=@n3.Shading_Gradient
   CINE_LightColor=@n3.Shading_Gradient
   ---------
   Color_Lit=LinearInterpolate(A=@n2.Color_Lit, B=LinearInterpolate(B=@n4, Alpha=@n12) {constAlpha=1, constB=0}, Alpha=@n15)
   Color_Shaded=LinearInterpolate(A=@n2.Color_Shaded, B=LinearInterpolate(B=@n16, Alpha=@n12) {constAlpha=1, constB=0}, Alpha=@n15)
   SRM=@n2.SRM
   Emissive=@n2.SRM
   DFL_Color=@n2.SRM
   Specular_Color=@n2.SRM
   RimLight_Color=@n2.SRM
  ) {materialFunction=/Game/Global/Materials/Character_Parent_2023/Anime_Parent_2023/MaterialFunctions/MFC_AnimeShader_2023_Post}
  NamedRerouteDeclaration(None=Multiply(A=ScalarParameter() {group=CountyFrog, parameterName=Drumkit_Blink, sliderMax=1}) {constB=0.6}) {name=Eyelids, nodeColor=<rgba 0 0.3437 1 1>}
 ]
}
